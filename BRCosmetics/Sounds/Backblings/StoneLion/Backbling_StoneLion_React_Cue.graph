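# Sound cue for the reaction sound of the StoneLion back bling cosmetic.
# A random node picks one of three equally weighted StoneLion react waves each time it plays.
# It plays at reduced volume in the back bling sound class, using close-range pet attenuation and the shared reactive back bling audio property sheet.
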SoundCue /BRCosmetics/Sounds/Backblings/StoneLion/Backbling_StoneLion_React_Cue
root=Random_0
nodes [
 {id=Random_0, children [WavePlayer_0,WavePlayer_1,WavePlayer_2], properties {weights [1,1,1]}}
 {id=WavePlayer_0, properties {soundWaveAssetPtr=&/Backpack_StoneLion_React_01}}
 {id=WavePlayer_1, properties {soundWaveAssetPtr=&/Backpack_StoneLion_React_02}}
 {id=WavePlayer_2, properties {soundWaveAssetPtr=&/Backpack_StoneLion_React_03}}
]
properties {
 attenuationSettings=&Fort_Pets_Close_Athena_Att
 audioPropertiesSheet=/BRCosmetics/Sounds/Settings/PropertySheets/Reactive/APS_Cosmetic_Reactive_Backbling
 maxDistance=1800.0001
 soundClassObject=/Game/Sounds/FortSoundClassesAndModes/Fort_Backblings
 volumeMultiplier=0.5
}
